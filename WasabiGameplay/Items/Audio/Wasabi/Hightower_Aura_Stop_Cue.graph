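# Hightower_Aura_Stop_Cue is a Sound Cue for a Wasabi item, named as the stop sound of the Hightower aura.
# Its graph has no nodes and no root, so it references no audio and has nothing to play.
# It still carries lowered pitch and volume, a max distance, and a concurrency override that caps simultaneous instances by stopping the quietest.
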SoundCue /WasabiGameplay/Items/Audio/Wasabi/Hightower_Aura_Stop_Cue
root=null
nodes []
properties {
 bOverrideConcurrency=true
 maxDistance=4000
 pitchMultiplier=0.6
 volumeMultiplier=0.5
 concurrencyOverrides {maxCount=3, resolutionRule=StopQuietest}
}
