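# Sound cue for the Wasabi item's dash sound, classed as melee weapon audio.
# A random node picks one of three dash wave variants with equal weight, and a modulator slightly varies pitch and volume on each play.
# It uses the Wasabi melee attenuation asset with HRTF spatialization and overrides concurrency to cap simultaneous voices.
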SoundCue /WasabiGameplay/Items/Audio/Wasabi/Wasabi_Dash_Dash_Cue
root=Modulator_1
nodes [
 {id=Modulator_1, children [Random_1], properties {pitchMax=1.1, pitchMin=1, volumeMax=1.1, volumeMin=1}}
 {id=Random_1, children [WavePlayer_3,WavePlayer_4,WavePlayer_5], properties {bShouldExcludeFromBranchCulling=true, weights [1,1,1]}}
 {id=WavePlayer_3, properties {soundWaveAssetPtr=&/Wasabi_Dash_Dash_01}}
 {id=WavePlayer_4, properties {soundWaveAssetPtr=&/Wasabi_Dash_Dash_02}}
 {id=WavePlayer_5, properties {soundWaveAssetPtr=&/Wasabi_Dash_Dash_03}}
]
properties {
 attenuationSettings=/Game/Sounds/Attenuation/Fort_Wasabi_Melee_Att
 bOverrideConcurrency=true
 maxDistance=3800
 pitchMultiplier=1.1
 soundClassObject=&Fort_WeaponMelee
 volumeMultiplier=0.3
 attenuationOverrides {spatializationAlgorithm=SPATIALIZATION_HRTF}
 concurrencyOverrides {maxCount=6}
 concurrencySet [null]
}
