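# Sound cue for the Wasabi dash slash, classed as a melee weapon sound.
# A Random node picks among four slash waves with equal weight and is excluded from branch culling.
# A root Modulator varies each play's volume slightly and shifts pitch only downward from normal.
# The cue uses the shared Wasabi melee attenuation asset and its own concurrency override to cap simultaneous instances.
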SoundCue /WasabiGameplay/Items/Audio/Wasabi/Wasabi_Dash_Slash_Cue
root=Modulator_1
nodes [
 {id=Modulator_1, children [Random_0], properties {pitchMax=1, pitchMin=0.85, volumeMax=1.1, volumeMin=1}}
 {
  id=Random_0
  children [WavePlayer_0,WavePlayer_1,WavePlayer_2,WavePlayer_6]
  properties {bShouldExcludeFromBranchCulling=true, weights [1,1,1,1]}
 }
 {id=WavePlayer_0, properties {soundWaveAssetPtr=&/Wasabi_Dash_Slash_01}}
 {id=WavePlayer_1, properties {soundWaveAssetPtr=&/Wasabi_Dash_Slash_02}}
 {id=WavePlayer_2, properties {soundWaveAssetPtr=&/Wasabi_Dash_Slash_03}}
 {id=WavePlayer_6, properties {soundWaveAssetPtr=&/Wasabi_Dash_Slash_04}}
]
properties {
 attenuationSettings=/Game/Sounds/Attenuation/Fort_Wasabi_Melee_Att
 bOverrideConcurrency=true
 maxDistance=3800
 soundClassObject=&Fort_WeaponMelee
 volumeMultiplier=0.3
 attenuationOverrides {spatializationAlgorithm=SPATIALIZATION_HRTF}
 concurrencyOverrides {maxCount=6}
 concurrencySet [null]
}
