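# Sound cue for the Wasabi item's dodge sound, played at reduced volume with a slightly raised pitch.
# A Random node picks one of three equally weighted dodge waves and feeds a Modulator left at default settings.
# The cue uses the melee weapon sound class and a close-range umbrella weapon attenuation, with HRTF spatialization set in its attenuation overrides.
# It shares the Wasabi dash global-limit concurrency to cap simultaneous playback.
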SoundCue /WasabiGameplay/Items/Audio/Wasabi/Wasabi_Dodge_Cue
root=Modulator_0
nodes [
 {id=Modulator_0, children [Random_0]}
 {id=Random_0, children [WavePlayer_0,WavePlayer_1,WavePlayer_2], properties {weights [1,1,1]}}
 {id=WavePlayer_0, properties {soundWaveAssetPtr=&/Wasabi_Dodge_01}}
 {id=WavePlayer_1, properties {soundWaveAssetPtr=&/Wasabi_Dodge_02}}
 {id=WavePlayer_2, properties {soundWaveAssetPtr=&/Wasabi_Dodge_03}}
]
properties {
 attenuationSettings=/Game/Sounds/Attenuation/Fort_Weapon_Umbrella_Close_Athena_Att
 maxDistance=4500
 pitchMultiplier=1.05
 soundClassObject=&Fort_WeaponMelee_PSM
 volumeMultiplier=0.25
 attenuationOverrides {spatializationAlgorithm=SPATIALIZATION_HRTF}
 concurrencySet [&/Wasabi_Dash_GlobalLimit_Conc]
}
