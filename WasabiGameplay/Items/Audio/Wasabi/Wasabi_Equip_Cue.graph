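# Sound cue for equipping the Wasabi melee weapon, playing one of three equip sound variations.
# A random node with equal weights picks the wave, then a modulator applies variation before output at reduced volume.
# It routes to the melee weapon sound class and overrides attenuation with distance falloff and low-pass filtering.
# Overridden concurrency caps simultaneous plays per owner, stopping the oldest.
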SoundCue /WasabiGameplay/Items/Audio/Wasabi/Wasabi_Equip_Cue
root=Modulator_0
nodes [
 {id=Modulator_0, children [Random_0]}
 {id=Random_0, children [WavePlayer_0,WavePlayer_2,WavePlayer_3], properties {bShouldExcludeFromBranchCulling=true, weights [1,1,1]}}
 {id=WavePlayer_0, properties {soundWaveAssetPtr=&/Wasabi_Equip_01}}
 {id=WavePlayer_2, properties {soundWaveAssetPtr=&/Wasabi_Equip_02}}
 {id=WavePlayer_3, properties {soundWaveAssetPtr=&/Wasabi_Equip_03}}
]
properties {
 attenuationSettings=/Game/Sounds/Attenuation/Fort_Wasabi_Melee_Att
 bOverrideAttenuation=true
 bOverrideConcurrency=true
 maxDistance=2600
 soundClassObject=&Fort_WeaponMelee
 volumeMultiplier=0.3
 attenuationOverrides {bAttenuateWithLPF=true, falloffDistance=2200, lPFFrequencyAtMax=500, lPFRadiusMax=3000, lPFRadiusMin=1000, stereoSpread=700}
 concurrencyOverrides {bLimitToOwner=true, maxCount=6, resolutionRule=StopOldest}
 concurrencySet [&Player_Melee_Weapon_Conc]
}
