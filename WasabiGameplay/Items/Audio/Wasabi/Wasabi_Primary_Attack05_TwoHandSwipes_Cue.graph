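# Sound cue for the Wasabi weapon's fifth primary attack, a two-handed swipe, in the melee weapon sound class.
# A random node picks one of three equally weighted wave variations, and a modulator varies its volume slightly on each play.
# The cue plays quieter and lower-pitched through a Wasabi melee attenuation asset, and overrides concurrency to cap simultaneous instances.
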SoundCue /WasabiGameplay/Items/Audio/Wasabi/Wasabi_Primary_Attack05_TwoHandSwipes_Cue
root=Modulator_1
nodes [
 {id=Modulator_1, children [Random_0], properties {volumeMax=1.1, volumeMin=0.9}}
 {id=Random_0, children [WavePlayer_0,WavePlayer_1,WavePlayer_2], properties {weights [1,1,1]}}
 {id=WavePlayer_0, properties {soundWaveAssetPtr=&/Wasabi_Primary_Attack05_TwoHandSwipes_01}}
 {id=WavePlayer_1, properties {soundWaveAssetPtr=&/Wasabi_Primary_Attack05_TwoHandSwipes_02}}
 {id=WavePlayer_2, properties {soundWaveAssetPtr=&/Wasabi_Primary_Attack05_TwoHandSwipes_03}}
]
properties {
 attenuationSettings=/Game/Sounds/Attenuation/Fort_Wasabi_Melee_Att
 bOverrideConcurrency=true
 maxDistance=3800
 pitchMultiplier=0.8
 soundClassObject=&Fort_WeaponMelee
 volumeMultiplier=0.4
 attenuationOverrides {spatializationAlgorithm=SPATIALIZATION_HRTF}
 concurrencyOverrides {maxCount=6}
 concurrencySet [null]
}
