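# Sound cue for the impact sound of Wasabi primary attack 01, routed to the melee weapon sound class with Wasabi melee attenuation.
# A quality-level root sends the first level to a weighted random pick of six waves and the other three to one shared, evenly weighted three-wave branch.
# Each branch adds slight pitch and volume modulation.
# The cue lowers overall pitch and volume and overrides concurrency to cap simultaneous plays.
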SoundCue /WasabiGameplay/Items/Audio/Wasabi/Wasabi_Primary_Attack_01_Impact_Cue
root=QualityLevel_0
nodes [
 {id=Modulator_1, children [Random_0], properties {pitchMin=0.9, volumeMax=1.1, volumeMin=0.9}}
 {id=Modulator_2, children [Random_2], properties {pitchMin=0.9, volumeMax=1.1, volumeMin=0.9}}
 {id=QualityLevel_0, children [Modulator_1,Modulator_2,Modulator_2,Modulator_2]}
 {
  id=Random_0
  children [WavePlayer_0, WavePlayer_1, WavePlayer_2, WavePlayer_9, WavePlayer_12, WavePlayer_13]
  properties {weights [2,2,1,2,1,1]}
 }
 {id=Random_2, children [WavePlayer_15, WavePlayer_16, WavePlayer_17], properties {bShouldExcludeFromBranchCulling=true, weights [1,1,1]}}
 {id=WavePlayer_0, properties {soundWaveAssetPtr=&/Wasabi_Primary_Attack_01_Impact_01}}
 {id=WavePlayer_1, properties {soundWaveAssetPtr=&/Wasabi_Primary_Attack_01_Impact_02}}
 {id=WavePlayer_12, properties {soundWaveAssetPtr=&/Wasabi_Primary_Attack_01_Impact_05}}
 {id=WavePlayer_13, properties {soundWaveAssetPtr=&/Wasabi_Primary_Attack_01_Impact_06}}
 {id=WavePlayer_15, properties {soundWaveAssetPtr=&Wasabi_Primary_Attack_01_Impact_01}}
 {id=WavePlayer_16, properties {soundWaveAssetPtr=&Wasabi_Primary_Attack_01_Impact_02}}
 {id=WavePlayer_17, properties {soundWaveAssetPtr=&/Wasabi_Primary_Attack_01_Impact_04}}
 {id=WavePlayer_2, properties {soundWaveAssetPtr=&/Wasabi_Primary_Attack_01_Impact_03}}
 {id=WavePlayer_9, properties {soundWaveAssetPtr=&Wasabi_Primary_Attack_01_Impact_04}}
]
properties {
 attenuationSettings=/Game/Sounds/Attenuation/Fort_Wasabi_Melee_Att
 bOverrideConcurrency=true
 maxDistance=3800
 pitchMultiplier=0.8
 soundClassObject=&Fort_WeaponMelee
 volumeMultiplier=0.3
 attenuationOverrides {spatializationAlgorithm=SPATIALIZATION_HRTF}
 concurrencyOverrides {maxCount=6}
 concurrencySet [null]
}
