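# Sound cue for the swing of the Wasabi melee weapon's primary attack, routed to the weapon melee sound class.
# A quality-level root sends its first level to a random pick of six swing waves and the other three levels to a shared three-wave pick, which is excluded from branch culling.
# Each branch modulates pitch and volume.
# Melee attenuation and a concurrency override govern playback.
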SoundCue /WasabiGameplay/Items/Audio/Wasabi/Wasabi_Primary_Attack_01_Swing_Cue
root=QualityLevel_0
nodes [
 {id=Modulator_1, children [Random_1], properties {pitchMin=0.9, volumeMax=1.1, volumeMin=0.9}}
 {id=Modulator_2, children [Random_2], properties {pitchMin=0.9, volumeMax=1.1, volumeMin=0.9}}
 {id=QualityLevel_0, children [Modulator_1,Modulator_2,Modulator_2,Modulator_2]}
 {id=Random_1, children [WavePlayer_3,WavePlayer_4,WavePlayer_5,WavePlayer_6,WavePlayer_7,WavePlayer_8], properties {weights [1,1,1,1,1,1]}}
 {id=Random_2, children [WavePlayer_10, WavePlayer_11, WavePlayer_14], properties {bShouldExcludeFromBranchCulling=true, weights [1,1,1]}}
 {id=WavePlayer_10, properties {soundWaveAssetPtr=&/Wasabi_Primary_Attack_01_Swing_01}}
 {id=WavePlayer_11, properties {soundWaveAssetPtr=&/Wasabi_Primary_Attack_01_Swing_02}}
 {id=WavePlayer_14, properties {soundWaveAssetPtr=&/Wasabi_Primary_Attack_01_Swing_05}}
 {id=WavePlayer_3, properties {soundWaveAssetPtr=&/Wasabi_Primary_Attack_01_Swing_06}}
 {id=WavePlayer_4, properties {soundWaveAssetPtr=&Wasabi_Primary_Attack_01_Swing_01}}
 {id=WavePlayer_5, properties {soundWaveAssetPtr=&Wasabi_Primary_Attack_01_Swing_02}}
 {id=WavePlayer_6, properties {soundWaveAssetPtr=&/Wasabi_Primary_Attack_01_Swing_03}}
 {id=WavePlayer_7, properties {soundWaveAssetPtr=&/Wasabi_Primary_Attack_01_Swing_04}}
 {id=WavePlayer_8, properties {soundWaveAssetPtr=&Wasabi_Primary_Attack_01_Swing_05}}
]
properties {
 attenuationSettings=/Game/Sounds/Attenuation/Fort_Wasabi_Melee_Att
 bOverrideConcurrency=true
 maxDistance=3800
 soundClassObject=&Fort_WeaponMelee
 volumeMultiplier=0.2
 attenuationOverrides {spatializationAlgorithm=SPATIALIZATION_HRTF}
 concurrencyOverrides {maxCount=6}
 concurrencySet [null]
}
