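# Sound cue for the body-lift layer of the Wasabi weapon's fourth primary melee attack.
# A Random node picks one of three equally weighted wave variants, feeding a Modulator that slightly varies volume per play.
# It routes to the melee weapon sound class at reduced volume with Wasabi melee attenuation.
# It also sets a concurrency override capping simultaneous instances, plus an HRTF spatialization value in its attenuation overrides.
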SoundCue /WasabiGameplay/Items/Audio/Wasabi/Wasabi_Primary_Attack_04_BodyLift_Cue
root=Modulator_1
nodes [
 {id=Modulator_1, children [Random_3], properties {volumeMax=1.1, volumeMin=0.9}}
 {id=Random_3, children [WavePlayer_9, WavePlayer_10, WavePlayer_11], properties {weights [1,1,1]}}
 {id=WavePlayer_10, properties {soundWaveAssetPtr=&/Wasabi_Primary_Attack_04_BodyLift_02}}
 {id=WavePlayer_11, properties {soundWaveAssetPtr=&/Wasabi_Primary_Attack_04_BodyLift_03}}
 {id=WavePlayer_9, properties {soundWaveAssetPtr=&/Wasabi_Primary_Attack_04_BodyLift_01}}
]
properties {
 attenuationSettings=/Game/Sounds/Attenuation/Fort_Wasabi_Melee_Att
 bOverrideConcurrency=true
 maxDistance=3800
 soundClassObject=&Fort_WeaponMelee
 volumeMultiplier=0.35
 attenuationOverrides {spatializationAlgorithm=SPATIALIZATION_HRTF}
 concurrencyOverrides {maxCount=6}
 concurrencySet [null]
}
