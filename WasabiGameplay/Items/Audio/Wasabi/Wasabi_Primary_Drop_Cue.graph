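# Sound cue for dropping the Wasabi primary item, layering a main drop sound with a flavour sound.
# A mixer blends a modulated random pick from three drop waves at full volume with a quieter random pick from two flavour waves.
# The cue plays slightly pitched down, using the weapon-drop attenuation and the loot-collect sound class.
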SoundCue /WasabiGameplay/Items/Audio/Wasabi/Wasabi_Primary_Drop_Cue
root=Mixer_0
nodes [
 {id=Mixer_0, children [Modulator_0,Random_0], properties {inputVolume [1,0.7]}}
 {id=Modulator_0, children [Random_1]}
 {id=Random_0, children [WavePlayer_5,WavePlayer_6], properties {weights [1,1]}}
 {id=Random_1, children [WavePlayer_1,WavePlayer_3,WavePlayer_4], properties {weights [1,1,1]}}
 {id=WavePlayer_1, properties {soundWaveAssetPtr=&/Wasabi_Primary_Drop_01}}
 {id=WavePlayer_3, properties {soundWaveAssetPtr=&/Wasabi_Primary_Drop_02}}
 {id=WavePlayer_4, properties {soundWaveAssetPtr=&/Wasabi_Primary_Drop_03}}
 {id=WavePlayer_5, properties {soundWaveAssetPtr=&/Wasabi_Primary_Flavour_Drop_02}}
 {id=WavePlayer_6, properties {soundWaveAssetPtr=&/Wasabi_Primary_Flavour_Drop_01}}
]
properties {
 attenuationSettings=&Fort_Weapon_Drop_Att
 maxDistance=3000
 pitchMultiplier=0.9
 soundClassObject=&Fort_Loot_Collect
 volumeMultiplier=1
 concurrencySet [null]
}
