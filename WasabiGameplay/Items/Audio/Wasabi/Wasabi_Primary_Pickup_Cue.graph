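# Sound cue that plays the pickup sound for the Wasabi primary item, classed as a loot collect sound.
# A single wave player for Wasabi_Primary_Pickup_01 feeds a modulator, with lowered cue volume and pitch.
# It uses a close-range umbrella weapon attenuation asset and overrides concurrency so only one instance plays, blocking new plays while one is active.
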SoundCue /WasabiGameplay/Items/Audio/Wasabi/Wasabi_Primary_Pickup_Cue
root=Modulator_0
nodes [
 {id=Modulator_0, children [WavePlayer_0]}
 {id=WavePlayer_0, properties {soundWaveAssetPtr=&/Wasabi_Primary_Pickup_01}}
]
properties {
 attenuationSettings=/Game/Sounds/Attenuation/Fort_Weapon_Umbrella_Close_Athena_Att
 bOverrideConcurrency=true
 maxDistance=4500
 pitchMultiplier=0.9
 soundClassObject=&Fort_Loot_Collect
 volumeMultiplier=0.199
 concurrencyOverrides {maxCount=1, resolutionRule=PreventNew}
}
