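# Sound cue that plays the retract sound for the Hightower Wasabi pickaxe melee weapon.
# A single wave feeds a close-range pickaxe attenuation node, then a modulator at the root.
# It plays at reduced volume in the melee weapon sound class, with a concurrency override limiting it to one instance per owner and stopping the oldest.
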SoundCue /WasabiGameplay/Items/Audio/Wasabi/Wasabi_Retract_Cue
root=Modulator_1
nodes [
 {id=Attenuation_2, children [WavePlayer_1], properties {attenuationSettings=/Game/Sounds/Attenuation/Fort_Weapon_Pickaxe_Close_Att}}
 {id=Modulator_1, children [Attenuation_2]}
 {id=WavePlayer_1, properties {soundWaveAssetPtr=/Game/Athena/Sounds/Weapons/PickAxes/Hightower_Wasabi/Pickaxe_HightowerWasabi_Retract_01}}
]
properties {
 bOverrideConcurrency=true
 maxDistance=1536
 soundClassObject=&Fort_WeaponMelee
 volumeMultiplier=0.2
 concurrencyOverrides {bLimitToOwner=true, maxCount=1, resolutionRule=StopOldest}
 concurrencySet [&Player_Melee_Weapon_Conc]
}
