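# Sound cue for the jump-spin sound of the Wasabi melee weapon's spin slash.
# A single wave player feeds a modulator that slightly randomizes pitch and volume on each play.
# It plays quietly in the weapon melee sound class with the shared Wasabi melee attenuation asset.
# The cue overrides concurrency to cap simultaneous instances; its attenuation overrides specify HRTF spatialization.
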SoundCue /WasabiGameplay/Items/Audio/Wasabi/Wasabi_SpinSlash_JumpSpin_Cue
root=Modulator_1
nodes [
 {id=Modulator_1, children [WavePlayer_1], properties {pitchMax=1, pitchMin=0.9, volumeMax=1.1, volumeMin=0.9}}
 {id=WavePlayer_1, properties {soundWaveAssetPtr=&/Wasabi_SpinSlash_JumpSpin_05}}
]
properties {
 attenuationSettings=/Game/Sounds/Attenuation/Fort_Wasabi_Melee_Att
 bOverrideConcurrency=true
 maxDistance=3800
 soundClassObject=&Fort_WeaponMelee
 volumeMultiplier=0.4
 attenuationOverrides {spatializationAlgorithm=SPATIALIZATION_HRTF}
 concurrencyOverrides {maxCount=6}
 concurrencySet [null]
}
